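; A helper function library for the quest debug tool, supplying display text and colours.
; GameplayTagToText turns a valid gameplay tag's debug string into text, and TextOrFallback passes text through; both substitute a fallback, "None" by default, for invalid or empty input.
; GetDebugListEntryTypeColor looks up a debug list entry type in a local enum-to-colour map and returns a zero value when the type is missing.
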(fn GameplayTagToText
  (params (Tag :type Gameplay%20Tag%20Structure%20%28by%20ref%29) (Fallback :type Text :default "NSLOCTEXT(\"[876A2DC0F464EFC64B90EF0730CEB03F]\", \"533D84E6481624CF032FD2B761D0DD90\", \"None\")"))
  (return (Utilities|String|SelectText (Utilities|Text|ToText(String) (GameplayTags|GetDebugStringfromGameplayTag Tag)) Fallback (GameplayTags|IsGameplayTagValid Tag))))

(fn GetDebugListEntryTypeColor
  (params (EntryType :type EQuestPanelDebugListEntryType%20Enum :default "Unset"))
  (bind _typecolors (LocalVariables|GetTypeColors))
  (bind (_value _returnvalue) (Utilities|Map|Find _typecolors EntryType))
  (if _returnvalue
    (return _value)
    (else
      (return 0))))

(fn TextOrFallback
  (params (Input :type Text) (Fallback :type Text :default "NSLOCTEXT(\"[876A2DC0F464EFC64B90EF0730CEB03F]\", \"813E618A48948C8F39B758BB672EE567\", \"None\")"))
  (return (select (Utilities|Text|TextIsEmpty Input) Fallback Input)))
