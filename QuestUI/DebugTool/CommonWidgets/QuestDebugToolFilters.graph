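# Filter bar widget for the quest debug tool, with a search field, content-type checkboxes and organization tag filters in an options panel that starts collapsed.
# Each input change writes into the ActiveFilter struct and broadcasts OnActiveFilterChanged.
# InitFilter restores the controls from a given filter and initializes the tag widget from a root organization node.
# A switcher shows a Build Organization Tags button, which fires OnBuildTagsRequested, until SetTagsAvailable reveals the tag filters.
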WidgetBlueprint /QuestUI/DebugTool/CommonWidgets/QuestDebugToolFilters
graphs [
 BroadcastFilterChanged {nodes [
   BroadcastFilterChanged() {is=title K2Node_FunctionEntry}
   > Call On Active Filter Changed!(
    execute:Exec
    self:Quest Debug Tool Filters Object Reference
    ActiveFilter:Quest Panel Debug Tool Filter Structure=Get ActiveFilter() -> (ActiveFilter:Quest Panel Debug Tool Filter Structure)
   ) -> (then:Exec) {is=title K2Node_CallDelegate}
  ]}
 EventGraph {nodes [
   n1 = "On Text Changed (InputField_SearchFilter)"!() -> (OutputDelegate, then, Text:Text) {is=title K2Node_ComponentBoundEvent}
   > n2 = Set members in Quest Panel Debug Tool Filter!(
    execute:Exec
    StructRef:"Quest Panel Debug Tool Filter Structure (by ref)"=Get ActiveFilter()
    SearchString:String="To String (Text)"(InText=@n1.Text)
   ) -> (then:Exec, StructOut:"Quest Panel Debug Tool Filter Structure (by ref)") {is=title K2Node_SetFieldsInStruct}
   BroadcastFilterChanged(execute=[@n15.then, @n2.then, @n4.then, @n6.then, @n17.then, @n10.then, @n13.then, @n8.then]) {is=title K2Node_CallFunction}
   n3 = "On State Changed (Checkbox_Quests)"!() -> (OutputDelegate, then, State:bool) {is=title K2Node_ComponentBoundEvent}
   > n4 = Set members in Quest Panel Debug Tool Filter!(execute, StructRef=Get ActiveFilter(), bIncludeQuests:bool=@n3.State) -> (then, StructOut)
   n5 = "On State Changed (Checkbox_Bundles)"!() -> (OutputDelegate, then, State:bool) {is=title K2Node_ComponentBoundEvent}
   > n6 = Set members in Quest Panel Debug Tool Filter!(execute, StructRef=Get ActiveFilter(), bIncludeBundles:bool=@n5.State) -> (then, StructOut)
   n7 = "On State Changed (Checkbox_Schedules)"!() -> (OutputDelegate, then, State:bool) {is=title K2Node_ComponentBoundEvent}
   > n8 = Set members in Quest Panel Debug Tool Filter!(execute, StructRef=Get ActiveFilter(), bIncludeSchedules:bool=@n7.State) -> (then, StructOut)
   n9 = "On State Changed (Checkbox_ViewsOnly)"!() -> (OutputDelegate, then, State:bool) {is=title K2Node_ComponentBoundEvent}
   > n10 = Set members in Quest Panel Debug Tool Filter!(execute, StructRef=Get ActiveFilter(), bViewOnly:bool=@n9.State) -> (then, StructOut)
   "On Clicked (Button_ExpandOptions)"!() -> (OutputDelegate, then, Button:Common Button Base Object Reference) {is=title K2Node_ComponentBoundEvent}
   n11 = Get Panel_Options() -> (Panel_Options:Vertical Box Object Reference)
   > ExpandOptions!(
    execute
    self
    Expand:bool="Equal (Enum)"(
     A:ESlateVisibility Enum=GetVisibility!(self:Widget Object Reference=@n11) -> (ReturnValue:ESlateVisibility Enum=Visible)
     B:ESlateVisibility Enum=Collapsed
    )
   ) -> (then)
   n12 = "On Active Tags Changed (Widget_OrganizationTags)"!() -> (OutputDelegate, then, ActiveTags:"Gameplay Tag Container Structure (by ref)") {is=title K2Node_ComponentBoundEvent}
   > n13 = Set members in Quest Panel Debug Tool Filter!(execute, StructRef=Get ActiveFilter(), OrganizationTags:Gameplay Tag Container Structure=@n12.ActiveTags) -> (then, StructOut)
   n14 = "On State Changed (Checkbox_UnorganizedContent)"!() -> (OutputDelegate, then, State:bool) {is=title K2Node_ComponentBoundEvent}
   > n15 = Set members in Quest Panel Debug Tool Filter!(execute, StructRef=Get ActiveFilter(), bIncludeUnorganizedContent:bool=@n14.State) -> (then, StructOut)
   "On Clicked (Button_BuildTags)"!() -> (OutputDelegate, then, Button:Common Button Base Object Reference) {is=title K2Node_ComponentBoundEvent}
   > Call On Build Tags Requested!(execute:Exec, self:Quest Debug Tool Filters Object Reference) -> (then:Exec) {is=title K2Node_CallDelegate}
   n16 = "On State Changed (Checkbox_Categories)"!() -> (OutputDelegate, then, State:bool) {is=title K2Node_ComponentBoundEvent}
   > n17 = Set members in Quest Panel Debug Tool Filter!(execute, StructRef=Get ActiveFilter(), bIncludeCategories:bool=@n16.State) -> (then, StructOut)
   Event Construct()
   > ExpandOptions(Expand=false)
  ]}
 ExpandOptions {nodes [
   n1 = ExpandOptions!() -> (then, Expand:bool=~) {is=title K2Node_FunctionEntry}
   > SetVisibility(self=Get Panel_Options(), InVisibility=Select(Option 0=Collapsed, Option 1=SelfHitTestInvisible, Index=@n1))
  ]}
 InitFilter {nodes [
   n1 = InitFilter!() -> (then, Filter:"Quest Panel Debug Tool Filter Structure (by ref)", RootOrganizationNode:Quest Panel Debug Organization Node Object Reference) {is=title K2Node_FunctionEntry}
   > n2 = Set ActiveFilter(ActiveFilter:Quest Panel Debug Tool Filter Structure=@n1) -> (Output_Get:Quest Panel Debug Tool Filter Structure)
   > n3 = Sequence!(execute) -> (then_0, then_1, then_2:Exec, then_3:Exec, then_4:Exec, then_5:Exec, then_6:Exec, then_7:Exec, then_8:Exec)
   n4 = Break Quest Panel Debug Tool Filter!(QuestPanelDebugToolFilter:"Quest Panel Debug Tool Filter Structure (by ref)"=@n2) -> (
    SearchString:String
    bIncludeQuests:bool=~
    bIncludeBundles:bool=~
    bIncludeSchedules:bool=~
    bIncludeCategories:bool=~
    bViewOnly:bool=~
    bIncludeUnorganizedContent:bool=~
    OrganizationTags:Gameplay Tag Container Structure
    bTagsAvailable:bool=~
   ) {is=title K2Node_BreakStruct}
   SetState!(
    execute=@n3.then_1
    self:Quest Debug Tool Checkbox Object Reference=Get Checkbox_Quests() -> (Checkbox_Quests:Quest Debug Tool Checkbox Object Reference)
    State:bool=@n4.bIncludeQuests
   ) -> (then)
   SetState(execute=@n3.then_2, self=Get Checkbox_Bundles() -> (Checkbox_Bundles:Quest Debug Tool Checkbox Object Reference), State=@n4.bIncludeBundles)
   SetState(
    execute=@n3.then_6
    self=Get Checkbox_UnorganizedContent() -> (Checkbox_UnorganizedContent:Quest Debug Tool Checkbox Object Reference)
    State=@n4.bIncludeUnorganizedContent
   )
   SetState(
    execute=@n3.then_4
    self=Get Checkbox_Categories() -> (Checkbox_Categories:Quest Debug Tool Checkbox Object Reference)
    State=@n4.bIncludeCategories
   )
   SetTagsAvailable!(execute=@n3.then_8, self, Available:bool=@n4.bTagsAvailable) -> (then)
   SetState(
    execute=@n3.then_3
    self=Get Checkbox_Schedules() -> (Checkbox_Schedules:Quest Debug Tool Checkbox Object Reference)
    State=@n4.bIncludeSchedules
   )
   SetText!(
    execute=@n3.then_0
    self:WBP UIKit Input Field Base Object Reference=Get InputField_SearchFilter() -> (InputField_SearchFilter:WBP UIKit Input Field Object Reference)
    Text:Text="To Text (String)"(InString=@n4)
   ) -> (then)
   SetState(execute=@n3.then_5, self=Get Checkbox_ViewsOnly() -> (Checkbox_ViewsOnly:Quest Debug Tool Checkbox Object Reference), State=@n4.bViewOnly)
   Init!(
    execute=@n3.then_7
    self:Quest Debug Tool Organization Filters Object Reference=Get Widget_OrganizationTags() -> (Widget_OrganizationTags:Quest Debug Tool Organization Filters Object Reference)
    RootNode:Quest Panel Debug Organization Node Object Reference=Get RootOrganizationNode() -> (RootOrganizationNode:Quest Panel Debug Organization Node Object Reference)
    AllCheckedTags:"Gameplay Tag Container Structure (by ref)"=@n4.OrganizationTags
   ) -> (then)
  ]}
 OnActiveFilterChanged {nodes [OnActiveFilterChanged!() -> (then, ActiveFilter:Quest Panel Debug Tool Filter Structure) {is=title K2Node_FunctionEntry}]}
 OnBuildTagsRequested {nodes [OnBuildTagsRequested() {is=title K2Node_FunctionEntry}]}
 SetTagsAvailable {nodes [
   n1 = SetTagsAvailable!() -> (then, Available:bool=~) {is=title K2Node_FunctionEntry}
   > SetActiveWidget(
    self=Get Switcher_Tags() -> (Switcher_Tags:Common Visibility Switcher Object Reference)
    Widget=Select(
     Option 0:Widget Object Reference=Get Button_BuildTags() -> (Button_BuildTags:WBP UIKit Button Quiet Object Reference)
     Option 1:Widget Object Reference=Get Panel_Tags() -> (Panel_Tags:Vertical Box Object Reference)
     Index=@n1
    ) -> (ReturnValue:Widget Object Reference)
   )
  ]}
]
variables [
 ActiveFilter {
  type=QuestPanelDebugToolFilter
  default {
   bIncludeBundles=true
   bIncludeCategories=true
   bIncludeQuests=true
   bIncludeSchedules=true
   bIncludeUnorganizedContent=true
   bTagsAvailable=false
   bViewOnly=false
   searchString=""
   organizationTags {gameplayTags []}
  }
 }
 OnActiveFilterChanged {}
 OnBuildTagsRequested {}
]
widgets {
 info {inheritedWidgetCount=0, parentClass=&UserWidget, rootWidgetClass=&VerticalBox, widgetCount=21}
 widgets [
  VerticalBox_38 {widgetClassPath=&VerticalBox, widget=VerticalBox_38}
  HorizontalBox_1 {widgetClassPath=&HorizontalBox, parent=VerticalBox_38, widget=HorizontalBox_1, slot {class=VerticalBoxSlot}}
  InputField_SearchFilter {
   widgetClassPath=/FortUILibrary/UIKit/InputField/WBP_UIKit_InputField
   bIsVariable=true
   parent=HorizontalBox_1
   widget=InputField_SearchFilter
   properties {hint Text=Search for asset, icon=/Game/UI/Foundation/Textures/Banner/Bucket/T-Banners-Icons-MagnifyingGlass, maxTextCharacters=60}
   slot {class=HorizontalBoxSlot, properties {size {sizeRule=Fill, value=1}}}
  }
  Button_ExpandOptions {
   widgetClassPath=&/QuestDebugToolButton
   bIsVariable=true
   parent=HorizontalBox_1
   widget=Button_ExpandOptions
   properties {
    bInteractableWhenSelected=true
    bSelectable=true
    bToggleable=true
    holdData=&CommonUI_HoldData_Base_C
    iconBrush {
     bIsDynamicallyLoaded=false
     drawAs=Image
     imageType=FullColor
     mirroring=NoMirror
     resourceName=None
     resourceObject=/FortUILibrary/Textures/MaterialInstances/MI_UIKit_Icon_GeneralSettings
     tiling=NoTile
     imageSize=<32 32>
     margin=<ltrb 0 0 0 0>
     outlineSettings {
      bUseBrushTransparency=false
      roundingType=HalfHeightRadius
      width=0
      color {colorUseRule=UseColor_Specified, specifiedColor=<rgba 0 0 0 0>}
      cornerRadii=<0 0 0 1>
     }
     tintColor {colorUseRule=UseColor_Specified, specifiedColor=<rgba 1 1 1 1>}
     uVRegion {bIsValid=false, max=<0 0>, min=<0 0>}
    }
   }
   slot {class=HorizontalBoxSlot, properties {padding=<ltrb 15 0 0 0>}}
  }
  Spacer_44 {widgetClassPath=&Spacer, parent=VerticalBox_38, widget=Spacer_44, properties {size=<1 12>}, slot {class=VerticalBoxSlot}}
  Panel_Options {widgetClassPath=&VerticalBox, bIsVariable=true, parent=VerticalBox_38, widget=Panel_Options, slot {class=VerticalBoxSlot}}
  HorizontalBox_0 {
   widgetClassPath=&HorizontalBox
   parent=Panel_Options
   widget=HorizontalBox_0
   slot {class=VerticalBoxSlot, properties {size {sizeRule=Fill, value=1}}}
  }
  Checkbox_Quests {
   widgetClassPath=&/QuestDebugToolCheckbox
   bIsVariable=true
   parent=HorizontalBox_0
   widget=Checkbox_Quests
   properties {label=Show Quests}
   slot {class=HorizontalBoxSlot, properties {size {sizeRule=Fill, value=1}}}
  }
  Checkbox_Bundles {
   widgetClassPath=&QuestDebugToolCheckbox
   bIsVariable=true
   parent=HorizontalBox_0
   widget=Checkbox_Bundles
   properties {label=Show Bundles}
   slot {class=HorizontalBoxSlot, properties {size {sizeRule=Fill, value=1}}}
  }
  HorizontalBox_73 {widgetClassPath=&HorizontalBox, parent=Panel_Options, widget=HorizontalBox_73, slot {class=VerticalBoxSlot}}
  Checkbox_Schedules {
   widgetClassPath=&QuestDebugToolCheckbox
   bIsVariable=true
   parent=HorizontalBox_73
   widget=Checkbox_Schedules
   properties {label=Show Schedules}
   slot {class=HorizontalBoxSlot, properties {size {sizeRule=Fill, value=1}}}
  }
  Checkbox_Categories {
   widgetClassPath=&QuestDebugToolCheckbox
   bIsVariable=true
   parent=HorizontalBox_73
   widget=Checkbox_Categories
   properties {label=Show Categories}
   slot {class=HorizontalBoxSlot, properties {size {sizeRule=Fill, value=1}}}
  }
  HorizontalBox_2 {
   widgetClassPath=&HorizontalBox
   parent=Panel_Options
   widget=HorizontalBox_2
   slot {class=VerticalBoxSlot, properties {size {sizeRule=Fill, value=1}}}
  }
  Checkbox_ViewsOnly {
   widgetClassPath=&QuestDebugToolCheckbox
   bIsVariable=true
   parent=HorizontalBox_2
   widget=Checkbox_ViewsOnly
   properties {label=Views Only}
   slot {class=HorizontalBoxSlot, properties {size {sizeRule=Fill, value=1}}}
  }
  Image_Spacer_Vert {
   widgetClassPath=&Image
   parent=Panel_Options
   widget=Image_Spacer_Vert
   properties {
    brush {
     bIsDynamicallyLoaded=false
     drawAs=Image
     imageType=NoImage
     mirroring=NoMirror
     resourceName=None
     resourceObject=null
     tiling=NoTile
     imageSize=<32 4>
     margin=<ltrb 0 0 0 0>
     outlineSettings {
      bUseBrushTransparency=false
      roundingType=HalfHeightRadius
      width=0
      color {colorUseRule=UseColor_Specified, specifiedColor=<rgba 0 0 0 0>}
      cornerRadii=<0 0 0 0>
     }
     tintColor {colorUseRule=UseColor_Specified, specifiedColor=<rgba 1 1 1 1>}
     uVRegion {bIsValid=false, max=<0 0>, min=<0 0>}
    }
    colorAndOpacity=<rgba 1 1 1 0.1>
   }
   slot {class=VerticalBoxSlot, properties {padding=<ltrb 8 8 8 8>}}
  }
  Switcher_Tags {
   widgetClassPath=&CommonVisibilitySwitcher
   bIsVariable=true
   parent=Panel_Options
   widget=Switcher_Tags
   slot {class=VerticalBoxSlot}
  }
  Button_BuildTags {
   widgetClassPath=/FortUILibrary/UIKit/Buttons/WBP_UIKit_Button_Quiet
   bIsVariable=true
   parent=Switcher_Tags
   widget=Button_BuildTags
   properties {minWidth=32, text=Build Organization Tags, toolTipText="Loads all assets to gather the organizational structure. May take ~30s to run."}
   slot {class=CommonVisibilitySwitcherSlot, properties {horizontalAlignment=HAlign_Center, verticalAlignment=VAlign_Center}}
  }
  Panel_Tags {
   widgetClassPath=&VerticalBox
   bIsVariable=true
   parent=Switcher_Tags
   widget=Panel_Tags
   slot {class=CommonVisibilitySwitcherSlot}
  }
  Checkbox_UnorganizedContent {
   widgetClassPath=&QuestDebugToolCheckbox
   bIsVariable=true
   parent=Panel_Tags
   widget=Checkbox_UnorganizedContent
   properties {label=Show Unorganized Content}
   slot {class=VerticalBoxSlot, properties {size {sizeRule=Fill, value=2}}}
  }
  Widget_OrganizationTags {
   widgetClassPath=&/QuestDebugToolOrganizationFilters
   bIsVariable=true
   parent=Panel_Tags
   widget=Widget_OrganizationTags
   slot {class=VerticalBoxSlot}
  }
  Image_Spacer_Vert_1 {
   widgetClassPath=&Image
   parent=VerticalBox_38
   widget=Image_Spacer_Vert_1
   properties {
    brush {
     bIsDynamicallyLoaded=false
     drawAs=Image
     imageType=NoImage
     mirroring=NoMirror
     resourceName=None
     resourceObject=null
     tiling=NoTile
     imageSize=<32 4>
     margin=<ltrb 0 0 0 0>
     outlineSettings {
      bUseBrushTransparency=false
      roundingType=HalfHeightRadius
      width=0
      color {colorUseRule=UseColor_Specified, specifiedColor=<rgba 0 0 0 0>}
      cornerRadii=<0 0 0 0>
     }
     tintColor {colorUseRule=UseColor_Specified, specifiedColor=<rgba 1 1 1 1>}
     uVRegion {bIsValid=false, max=<0 0>, min=<0 0>}
    }
    colorAndOpacity=<rgba 1 1 1 0.1>
   }
   slot {class=VerticalBoxSlot, properties {padding=<ltrb 8 8 8 8>}}
  }
 ]
}
