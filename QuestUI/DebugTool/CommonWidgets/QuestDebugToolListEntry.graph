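# List entry widget for the quest debug tool, built on QuestPanelDebugToolListEntry, showing a name with a colored pip over a background.
# Populate sets the display name, tints the pip by entry type color, and updates the temporary marker.
# SetIsTemp also toggles that "Temporary (Excluded by Filter)" label.
# Item assignment and selection changes run UpdateState, which inverts background and text colors for selected entries.
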WidgetBlueprint /QuestUI/DebugTool/CommonWidgets/QuestDebugToolListEntry
graphs [
 EventGraph {nodes [
   n1 = Event SetIsTemp!() -> (OutputDelegate, then, bIsTemp:bool) {is=title K2Node_Event}
   > UpdateIsTemp!(execute, self, IsTemp:bool=@n1.bIsTemp) -> (then)
   n2 = Event Populate!() -> (OutputDelegate, then, DisplayName:String, EntryType:EQuestPanelDebugListEntryType Enum=Unset, bIsTemp:bool) {is=title K2Node_Event}
   > "SetText (Text)"(self=Get Text_DisplayName() -> (Text_DisplayName:Common Text Object Reference), InText="To Text (String)"(InString=@n2.DisplayName))
   > SetColorAndOpacity(
    self:Image Object Reference=Get Pip() -> (Pip:Image Object Reference)
    InColorAndOpacity:Linear Color Structure=Get Debug List Entry Type Color!(EntryType:EQuestPanelDebugListEntryType Enum=@n2.EntryType) -> (Color:Linear Color Structure) {is=title K2Node_CallFunction}
   )
   > UpdateIsTemp(IsTemp=@n2.bIsTemp)
   n3 = Event OnListItemObjectSet()
   UpdateState(execute=[@n3.then, @n4.then])
   n4 = Event On Item Selection Changed!() -> (OutputDelegate, then, bIsSelected:bool) {is=title K2Node_Event}
  ]}
 UpdateIsTemp {nodes [
   n1 = UpdateIsTemp!() -> (then, IsTemp:bool=~) {is=title K2Node_FunctionEntry}
   > SetVisibility(
    self=Get Text_Temp() -> (Text_Temp:Common Text Object Reference)
    InVisibility=Select(Option 0=Collapsed, Option 1=HitTestInvisible, Index=@n1)
   )
  ]}
 UpdateState {nodes [
   UpdateState() {is=title K2Node_FunctionEntry}
   > n1 = Branch(Condition=IsListItemSelected!(UserListEntry:User List Entry Interface) -> (ReturnValue:bool))
   > SetColorAndOpacity(self=Get BG() -> (BG:Image Object Reference), InColorAndOpacity=(R=1,G=1,B=1,A=1))
   > SetColorAndOpacity(
    self:Text Object Reference=Get Text_DisplayName()
    InColorAndOpacity:Slate Color Structure=Make Slate Color(SpecifiedColor=(R=0,G=0,B=0,A=1))
   )
   SetColorAndOpacity(execute=@n1.else, self:Image Object Reference=Get BG(), InColorAndOpacity:Linear Color Structure=(R=0.0353,G=0.0353,B=0.0353,A=1))
   > SetColorAndOpacity(
    self:Text Object Reference=Get Text_DisplayName()
    InColorAndOpacity:Slate Color Structure=Make Slate Color(SpecifiedColor=(R=1,G=1,B=1,A=1))
   )
  ]}
]
variables []
widgets {
 info {inheritedWidgetCount=0, parentClass=/Script/QuestUI.QuestPanelDebugToolListEntry, rootWidgetClass=&VerticalBox, widgetCount=8}
 widgets [
  VerticalBox_0 {widgetClassPath=&VerticalBox, widget=VerticalBox_0}
  Text_Temp {
   widgetClassPath=&CommonTextBlock
   bIsVariable=true
   parent=VerticalBox_0
   widget=Text_Temp
   properties {
    text="Temporary (Excluded by Filter)"
    visibility=Collapsed
    font {
     bForceMonospaced=false
     bMaterialIsStencil=false
     fontMaterial=null
     fontObject=&NotoSans
     letterSpacing=0
     monospacedWidth=1
     size=9
     skewAmount=0
     typefaceFontName=NotoSans-Bold
     outlineSettings {
      bApplyOutlineToDropShadows=false
      bMiteredCorners=false
      bSeparateFillAlpha=false
      outlineMaterial=null
      outlineSize=0
      outlineColor=<rgba 0 0 0 1>
     }
    }
    shadowOffset=<0 0>
   }
   slot {class=VerticalBoxSlot, properties {horizontalAlignment=HAlign_Center, verticalAlignment=VAlign_Center}}
  }
  Overlay_19 {widgetClassPath=&Overlay, parent=VerticalBox_0, widget=Overlay_19, slot {class=VerticalBoxSlot}}
  BG {
   widgetClassPath=&Image
   bIsVariable=true
   parent=Overlay_19
   widget=BG
   properties {colorAndOpacity=<rgba 0.0365 0.0365 0.0365 1>}
   slot {class=OverlaySlot, properties {horizontalAlignment=HAlign_Fill, verticalAlignment=VAlign_Fill}}
  }
  HorizontalBox_0 {
   widgetClassPath=&HorizontalBox
   parent=Overlay_19
   widget=HorizontalBox_0
   slot {class=OverlaySlot, properties {horizontalAlignment=HAlign_Fill, verticalAlignment=VAlign_Fill}}
  }
  Pip {
   widgetClassPath=&Image
   bIsVariable=true
   parent=HorizontalBox_0
   widget=Pip
   properties {brush {
     bIsDynamicallyLoaded=false
     drawAs=Image
     imageType=NoImage
     mirroring=NoMirror
     resourceName=None
     resourceObject=null
     tiling=NoTile
     imageSize=<8 8>
     margin=<ltrb 0 0 0 0>
     outlineSettings {
      bUseBrushTransparency=false
      roundingType=HalfHeightRadius
      width=0
      color {colorUseRule=UseColor_Specified, specifiedColor=<rgba 0 0 0 0>}
      cornerRadii=<0 0 0 0>
     }
     tintColor {colorUseRule=UseColor_Specified, specifiedColor=<rgba 1 1 1 1>}
     uVRegion {bIsValid=false, max=<0 0>, min=<0 0>}
    }}
   slot {class=HorizontalBoxSlot, properties {horizontalAlignment=HAlign_Left}}
  }
  ScaleBox_0 {
   widgetClassPath=&ScaleBox
   parent=HorizontalBox_0
   widget=ScaleBox_0
   slot {class=HorizontalBoxSlot, properties {padding=<ltrb 4 4 10 4>}}
  }
  Text_DisplayName {
   widgetClassPath=&CommonTextBlock
   bIsVariable=true
   parent=ScaleBox_0
   widget=Text_DisplayName
   properties {
    style=/FortUILibrary/Styles/Typography/Foundation/TextStyle-Base-XXS-B.TextStyle-Base-XXS-B_C
    text=Entry Name
    font {
     bForceMonospaced=false
     bMaterialIsStencil=false
     fontMaterial=null
     fontObject=&NotoSans
     letterSpacing=0
     monospacedWidth=1
     size=14
     skewAmount=0
     typefaceFontName=NotoSans-Bold
     outlineSettings {
      bApplyOutlineToDropShadows=false
      bMiteredCorners=false
      bSeparateFillAlpha=false
      outlineMaterial=null
      outlineSize=0
      outlineColor=<rgba 0 0 0 1>
     }
    }
    shadowOffset=<0 0>
   }
   slot {class=ScaleBoxSlot, properties {horizontalAlignment=HAlign_Fill}}
  }
 ]
}
