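# Header bar widget for the Quest UI debug tool, built on QuestPanelDebugToolWidgetBase.
# A dark overlay holds the "Quest UI Debugger" title with right-aligned account, backend and calendar offset text fields.
# Three functions format a passed-in account name, backend name or DateTime, converted from UTC, into labeled text.
# The event graph is empty, and the file does not show how those results reach the text blocks.
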WidgetBlueprint /QuestUI/DebugTool/CommonWidgets/QuestDebugToolRootHeader
graphs [
 EventGraph {links [], nodes []}
 GetAccountName {nodes [
   n1 = GetAccountName!() -> (then, AccountName:String) {is=title K2Node_FunctionEntry}
   > Return Node!(execute, OutAccountName:Text=Format Text!(Format="LOCTEXT(\"Account: {0}\")", 0:String=@n1) -> (Result))
  ]}
 GetBackendName {nodes [
   n1 = GetBackendName!() -> (then, BackendName:String) {is=title K2Node_FunctionEntry}
   > Return Node!(execute, OutBackendName:Text=Format Text(Format="LOCTEXT(\"Backend: {0}\")", 0=@n1))
  ]}
 GetCalendarOffset {nodes [
   n1 = GetCalendarOffset!() -> (then, DateTime:Date Time Structure) {is=title K2Node_FunctionEntry}
   > Return Node!(
    execute
    OffsetText:Text=Format Text(
     Format="LOCTEXT(\"Current Offset: {0}\")"
     0:Text="As DateTime with Timezone (from UTC)"!(self:Quest Panel Debug Tool Widget Base Object Reference, InDateTime:"Date Time Structure (by ref)"=@n1) -> (ReturnValue:Text)
    )
   )
  ]}
]
variables []
widgets {
 info {inheritedWidgetCount=0, parentClass=/Script/QuestUI.QuestPanelDebugToolWidgetBase, rootWidgetClass=&Overlay, widgetCount=11}
 widgets [
  Overlay_126 {widgetClassPath=&Overlay, widget=Overlay_126}
  Image_30 {
   widgetClassPath=&Image
   bIsVariable=true
   parent=Overlay_126
   widget=Image_30
   properties {colorAndOpacity=<rgba 0.0208 0.0208 0.0208 1>}
   slot {class=OverlaySlot, properties {horizontalAlignment=HAlign_Fill, verticalAlignment=VAlign_Fill}}
  }
  HorizontalBox_70 {
   widgetClassPath=&HorizontalBox
   parent=Overlay_126
   widget=HorizontalBox_70
   slot {class=OverlaySlot, properties {horizontalAlignment=HAlign_Fill, verticalAlignment=VAlign_Fill, padding=<ltrb 10 10 10 10>}}
  }
  Text_Header {
   widgetClassPath=&CommonTextBlock
   bIsVariable=true
   parent=HorizontalBox_70
   widget=Text_Header
   properties {
    style=/FortUILibrary/Styles/Typography/Foundation/TextStyle-Base-S-B.TextStyle-Base-S-B_C
    text=Quest UI Debugger
    font {
     bForceMonospaced=false
     bMaterialIsStencil=false
     fontMaterial=null
     fontObject=&NotoSans
     letterSpacing=0
     monospacedWidth=1
     size=18
     skewAmount=0
     typefaceFontName=NotoSans-Bold
     outlineSettings {
      bApplyOutlineToDropShadows=false
      bMiteredCorners=false
      bSeparateFillAlpha=false
      outlineMaterial=null
      outlineSize=0
      outlineColor=<rgba 0 0 0 1>
     }
    }
    shadowOffset=<0 0>
   }
   slot {class=HorizontalBoxSlot, properties {verticalAlignment=VAlign_Center, size {sizeRule=Fill, value=1}}}
  }
  Spacer_80 {widgetClassPath=&Spacer, parent=HorizontalBox_70, widget=Spacer_80, properties {size=<25 1>}, slot {class=HorizontalBoxSlot}}
  HorizontalBox_55 {
   widgetClassPath=&HorizontalBox
   parent=HorizontalBox_70
   widget=HorizontalBox_55
   slot {class=HorizontalBoxSlot, properties {horizontalAlignment=HAlign_Right}}
  }
  Text_Account {
   widgetClassPath=&CommonTextBlock
   bIsVariable=true
   parent=HorizontalBox_55
   widget=Text_Account
   properties {
    justification=Center
    style=/FortUILibrary/Styles/Typography/Foundation/TextStyle-Base-XXS-B.TextStyle-Base-XXS-B_C
    text="{Account}"
    font {
     bForceMonospaced=false
     bMaterialIsStencil=false
     fontMaterial=null
     fontObject=&NotoSans
     letterSpacing=0
     monospacedWidth=1
     size=14
     skewAmount=0
     typefaceFontName=NotoSans-Bold
     outlineSettings {
      bApplyOutlineToDropShadows=false
      bMiteredCorners=false
      bSeparateFillAlpha=false
      outlineMaterial=null
      outlineSize=0
      outlineColor=<rgba 0 0 0 1>
     }
    }
    shadowOffset=<0 0>
   }
   slot {class=HorizontalBoxSlot, properties {verticalAlignment=VAlign_Center}}
  }
  Spacer_1 {widgetClassPath=&Spacer, parent=HorizontalBox_55, widget=Spacer_1, properties {size=<25 1>}, slot {class=HorizontalBoxSlot}}
  Text_Backend {
   widgetClassPath=&CommonTextBlock
   bIsVariable=true
   parent=HorizontalBox_55
   widget=Text_Backend
   properties {
    justification=Center
    style=&TextStyle-Base-XXS-B_C
    text="{Backend}"
    font {
     bForceMonospaced=false
     bMaterialIsStencil=false
     fontMaterial=null
     fontObject=&NotoSans
     letterSpacing=0
     monospacedWidth=1
     size=14
     skewAmount=0
     typefaceFontName=NotoSans-Bold
     outlineSettings {
      bApplyOutlineToDropShadows=false
      bMiteredCorners=false
      bSeparateFillAlpha=false
      outlineMaterial=null
      outlineSize=0
      outlineColor=<rgba 0 0 0 1>
     }
    }
    shadowOffset=<0 0>
   }
   slot {class=HorizontalBoxSlot, properties {verticalAlignment=VAlign_Center}}
  }
  Spacer {widgetClassPath=&Spacer, parent=HorizontalBox_55, widget=Spacer, properties {size=<25 1>}, slot {class=HorizontalBoxSlot}}
  Text_Offset {
   widgetClassPath=&CommonTextBlock
   bIsVariable=true
   parent=HorizontalBox_55
   widget=Text_Offset
   properties {
    justification=Right
    style=&TextStyle-Base-XXS-B_C
    text="{Calendar Offset}"
    font {
     bForceMonospaced=false
     bMaterialIsStencil=false
     fontMaterial=null
     fontObject=&NotoSans
     letterSpacing=0
     monospacedWidth=1
     size=14
     skewAmount=0
     typefaceFontName=NotoSans-Bold
     outlineSettings {
      bApplyOutlineToDropShadows=false
      bMiteredCorners=false
      bSeparateFillAlpha=false
      outlineMaterial=null
      outlineSize=0
      outlineColor=<rgba 0 0 0 1>
     }
    }
    shadowOffset=<0 0>
   }
   slot {class=HorizontalBoxSlot, properties {verticalAlignment=VAlign_Center}}
  }
 ]
}
