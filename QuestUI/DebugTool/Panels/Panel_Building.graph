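# Quest UI debug tool panel widget that shows a centered "Building Directory" loading screen with a throbber while the tool's directory builds.
# Its Update function takes a build state enum plus current and max progress, then writes a formatted count and percentage to Text_Progress.
# A Select on the enum maps each build stage, from initializing through finalizing and complete, to the label shown in Text_State.
# The event graph is empty.
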WidgetBlueprint /QuestUI/DebugTool/Panels/Panel_Building
graphs [
 EventGraph {links [], nodes []}
 Update {nodes [
   n1 = Update!() -> (then, State:EQuestPanelDebugToolBuildState Enum=Init, CurrentProgress:int, MaxProgress:int) {is=title K2Node_FunctionEntry}
   n2 = float / float(A="To Float (Integer)"(InInt=@n1.CurrentProgress), B="To Float (Integer)"(InInt=@n1.MaxProgress))
   > "SetText (Text)"(
    self=Get Text_Progress() -> (Text_Progress:Common Text Object Reference)
    InText=Format Text!(
     Format="LOCTEXT(\"{Current} / {Max} ({Percent}%)\")"
     Current:int=@n1.CurrentProgress
     Max:int=@n1.MaxProgress
     Percent:double=float * float(A=@n2, B=100) {is=title K2Node_PromotableOperator}
    ) -> (Result)
   )
   > "SetText (Text)"(
    self=Get Text_State() -> (Text_State:Common Text Object Reference)
    InText=Select!(
     Init:Text="LOCTEXT(\"Initializing\")"
     Schedules:Text="LOCTEXT(\"Processing schedules\")"
     Bundles:Text="LOCTEXT(\"Processing bundles\")"
     Quests:Text="LOCTEXT(\"Processing quests\")"
     Categories:Text="LOCTEXT(\"Processing categories\")"
     Views:Text="LOCTEXT(\"Processing views\")"
     Finalize:Text="LOCTEXT(\"Finalizing\")"
     Inactive:Text="LOCTEXT(\"Complete\")"
     Tags:Text="LOCTEXT(\"Processing tags\")"
     Index:EQuestPanelDebugToolBuildState Enum=@n1
    ) -> (ReturnValue:Text)
   )
  ]}
]
variables []
widgets {
 info {inheritedWidgetCount=0, parentClass=&UserWidget, rootWidgetClass=&Overlay, widgetCount=6}
 widgets [
  Overlay_0 {widgetClassPath=&Overlay, widget=Overlay_0}
  VerticalBox_26 {
   widgetClassPath=&VerticalBox
   parent=Overlay_0
   widget=VerticalBox_26
   slot {class=OverlaySlot, properties {horizontalAlignment=HAlign_Center, verticalAlignment=VAlign_Center}}
  }
  CircularThrobber_172 {
   widgetClassPath=/Script/UMG.CircularThrobber
   bIsVariable=true
   parent=VerticalBox_26
   widget=CircularThrobber_172
   slot {class=VerticalBoxSlot, properties {horizontalAlignment=HAlign_Center, verticalAlignment=VAlign_Center}}
  }
  Text_Header {
   widgetClassPath=&CommonTextBlock
   parent=VerticalBox_26
   widget=Text_Header
   properties {
    style=/FortUILibrary/Styles/Typography/Foundation/TextStyle-Base-L-B-Yellow.TextStyle-Base-L-B-Yellow_C
    text=Building Directory
    colorAndOpacity {colorUseRule=UseColor_Specified, specifiedColor=<rgba 1 0.8714 0.0723 1>}
    font {
     bForceMonospaced=false
     bMaterialIsStencil=false
     fontMaterial=null
     fontObject=&NotoSans
     letterSpacing=0
     monospacedWidth=1
     size=32
     skewAmount=0
     typefaceFontName=NotoSans-Bold
     outlineSettings {
      bApplyOutlineToDropShadows=false
      bMiteredCorners=false
      bSeparateFillAlpha=false
      outlineMaterial=null
      outlineSize=0
      outlineColor=<rgba 0 0 0 1>
     }
    }
    shadowOffset=<0 0>
   }
   slot {class=VerticalBoxSlot, properties {verticalAlignment=VAlign_Center}}
  }
  Text_Progress {
   widgetClassPath=&CommonTextBlock
   bIsVariable=true
   parent=VerticalBox_26
   widget=Text_Progress
   properties {
    style=/FortUILibrary/Styles/Typography/Foundation/TextStyle-Base-M-B.TextStyle-Base-M-B_C
    text="0 / 100 (0%)"
    font {
     bForceMonospaced=false
     bMaterialIsStencil=false
     fontMaterial=null
     fontObject=&NotoSans
     letterSpacing=0
     monospacedWidth=1
     size=24
     skewAmount=0
     typefaceFontName=NotoSans-Bold
     outlineSettings {
      bApplyOutlineToDropShadows=false
      bMiteredCorners=false
      bSeparateFillAlpha=false
      outlineMaterial=null
      outlineSize=0
      outlineColor=<rgba 0 0 0 1>
     }
    }
    shadowOffset=<0 0>
   }
   slot {class=VerticalBoxSlot, properties {horizontalAlignment=HAlign_Center, verticalAlignment=VAlign_Center}}
  }
  Text_State {
   widgetClassPath=&CommonTextBlock
   bIsVariable=true
   parent=VerticalBox_26
   widget=Text_State
   properties {
    style=&TextStyle-Base-M-B_C
    text=Current State
    font {
     bForceMonospaced=false
     bMaterialIsStencil=false
     fontMaterial=null
     fontObject=&NotoSans
     letterSpacing=0
     monospacedWidth=1
     size=24
     skewAmount=0
     typefaceFontName=NotoSans-Bold
     outlineSettings {
      bApplyOutlineToDropShadows=false
      bMiteredCorners=false
      bSeparateFillAlpha=false
      outlineMaterial=null
      outlineSize=0
      outlineColor=<rgba 0 0 0 1>
     }
    }
    shadowOffset=<0 0>
   }
   slot {class=VerticalBoxSlot, properties {horizontalAlignment=HAlign_Center, verticalAlignment=VAlign_Center}}
  }
 ]
}
