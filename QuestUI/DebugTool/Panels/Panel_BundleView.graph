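# Layout-only UserWidget panel for the quest debug tool that presents the bundle view.
# A BundleView type header tops a vertical stack: bundle info beside schedule info, punch card info beside plugin info, then a bundle entry info section.
# Faint image dividers separate the sections.
# The event graph is empty, so the panel itself has no logic feeding data to its child widgets.
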WidgetBlueprint /QuestUI/DebugTool/Panels/Panel_BundleView
graphs [EventGraph {links [], nodes []}]
variables []
widgets {
 info {inheritedWidgetCount=0, parentClass=&UserWidget, rootWidgetClass=&VerticalBox, widgetCount=13}
 widgets [
  Panel_Root {widgetClassPath=&VerticalBox, bIsVariable=true, widget=Panel_Root}
  TypeHeader {
   widgetClassPath=/QuestUI/DebugTool/CommonWidgets/QuestDebugToolTypeHeader
   parent=Panel_Root
   widget=TypeHeader
   properties {debugEntryType=BundleView}
   slot {class=VerticalBoxSlot}
  }
  HorizontalBox_0 {widgetClassPath=&HorizontalBox, parent=Panel_Root, widget=HorizontalBox_0, slot {class=VerticalBoxSlot}}
  Widget_BundleInfo {
   widgetClassPath=/QuestUI/DebugTool/BundleWidgets/QuestDebugToolBundleInfo
   bIsVariable=true
   parent=HorizontalBox_0
   widget=Widget_BundleInfo
   slot {class=HorizontalBoxSlot, properties {size {sizeRule=Fill, value=1}}}
  }
  Image_Spacer_Vert_2 {
   widgetClassPath=&Image
   parent=HorizontalBox_0
   widget=Image_Spacer_Vert_2
   properties {
    brush {
     bIsDynamicallyLoaded=false
     drawAs=Image
     imageType=NoImage
     mirroring=NoMirror
     resourceName=None
     resourceObject=null
     tiling=NoTile
     imageSize=<4 32>
     margin=<ltrb 0 0 0 0>
     outlineSettings {
      bUseBrushTransparency=false
      roundingType=HalfHeightRadius
      width=0
      color {colorUseRule=UseColor_Specified, specifiedColor=<rgba 0 0 0 0>}
      cornerRadii=<0 0 0 0>
     }
     tintColor {colorUseRule=UseColor_Specified, specifiedColor=<rgba 1 1 1 1>}
     uVRegion {bIsValid=false, max=<0 0>, min=<0 0>}
    }
    colorAndOpacity=<rgba 1 1 1 0.1>
   }
   slot {class=HorizontalBoxSlot, properties {padding=<ltrb 8 8 8 8>}}
  }
  Widget_ScheduleInfo {
   widgetClassPath=/QuestUI/DebugTool/ScheduleWidgets/QuestDebugToolScheduleInfo
   bIsVariable=true
   parent=HorizontalBox_0
   widget=Widget_ScheduleInfo
   slot {class=HorizontalBoxSlot, properties {size {sizeRule=Fill, value=1}}}
  }
  Image_Spacer_Horiz {
   widgetClassPath=&Image
   parent=Panel_Root
   widget=Image_Spacer_Horiz
   properties {
    brush {
     bIsDynamicallyLoaded=false
     drawAs=Image
     imageType=NoImage
     mirroring=NoMirror
     resourceName=None
     resourceObject=null
     tiling=NoTile
     imageSize=<32 4>
     margin=<ltrb 0 0 0 0>
     outlineSettings {
      bUseBrushTransparency=false
      roundingType=HalfHeightRadius
      width=0
      color {colorUseRule=UseColor_Specified, specifiedColor=<rgba 0 0 0 0>}
      cornerRadii=<0 0 0 0>
     }
     tintColor {colorUseRule=UseColor_Specified, specifiedColor=<rgba 1 1 1 1>}
     uVRegion {bIsValid=false, max=<0 0>, min=<0 0>}
    }
    colorAndOpacity=<rgba 1 1 1 0.1>
   }
   slot {class=VerticalBoxSlot, properties {padding=<ltrb 8 8 8 8>}}
  }
  HorizontalBox_2 {widgetClassPath=&HorizontalBox, parent=Panel_Root, widget=HorizontalBox_2, slot {class=VerticalBoxSlot}}
  Widget_PunchCardInfo {
   widgetClassPath=/QuestUI/DebugTool/BundleWidgets/QuestDebugToolPunchCardInfo
   bIsVariable=true
   parent=HorizontalBox_2
   widget=Widget_PunchCardInfo
   slot {class=HorizontalBoxSlot, properties {size {sizeRule=Fill, value=1}}}
  }
  Image_Spacer_Vert {
   widgetClassPath=&Image
   parent=HorizontalBox_2
   widget=Image_Spacer_Vert
   properties {
    brush {
     bIsDynamicallyLoaded=false
     drawAs=Image
     imageType=NoImage
     mirroring=NoMirror
     resourceName=None
     resourceObject=null
     tiling=NoTile
     imageSize=<4 32>
     margin=<ltrb 0 0 0 0>
     outlineSettings {
      bUseBrushTransparency=false
      roundingType=HalfHeightRadius
      width=0
      color {colorUseRule=UseColor_Specified, specifiedColor=<rgba 0 0 0 0>}
      cornerRadii=<0 0 0 0>
     }
     tintColor {colorUseRule=UseColor_Specified, specifiedColor=<rgba 1 1 1 1>}
     uVRegion {bIsValid=false, max=<0 0>, min=<0 0>}
    }
    colorAndOpacity=<rgba 1 1 1 0.1>
   }
   slot {class=HorizontalBoxSlot, properties {padding=<ltrb 8 8 8 8>}}
  }
  Widget_PluginInfo {
   widgetClassPath=/QuestUI/DebugTool/CommonWidgets/QuestDebugToolPluginInfo
   bIsVariable=true
   parent=HorizontalBox_2
   widget=Widget_PluginInfo
   slot {class=HorizontalBoxSlot, properties {size {sizeRule=Fill, value=1}}}
  }
  Image_Spacer_Horiz_1 {
   widgetClassPath=&Image
   parent=Panel_Root
   widget=Image_Spacer_Horiz_1
   properties {
    brush {
     bIsDynamicallyLoaded=false
     drawAs=Image
     imageType=NoImage
     mirroring=NoMirror
     resourceName=None
     resourceObject=null
     tiling=NoTile
     imageSize=<32 4>
     margin=<ltrb 0 0 0 0>
     outlineSettings {
      bUseBrushTransparency=false
      roundingType=HalfHeightRadius
      width=0
      color {colorUseRule=UseColor_Specified, specifiedColor=<rgba 0 0 0 0>}
      cornerRadii=<0 0 0 0>
     }
     tintColor {colorUseRule=UseColor_Specified, specifiedColor=<rgba 1 1 1 1>}
     uVRegion {bIsValid=false, max=<0 0>, min=<0 0>}
    }
    colorAndOpacity=<rgba 1 1 1 0.1>
   }
   slot {class=VerticalBoxSlot, properties {padding=<ltrb 8 8 8 8>}}
  }
  Widget_BundleEntryInfo {
   widgetClassPath=/QuestUI/DebugTool/BundleWidgets/QuestDebugToolBundleEntryInfo
   bIsVariable=true
   parent=Panel_Root
   widget=Widget_BundleEntryInfo
   slot {class=VerticalBoxSlot}
  }
 ]
}
