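# Static help panel widget for the QuestUI debug tool, explaining why a quest-related asset may be missing from the tool's list.
# An Overlay root centers one padded, auto-wrapping CommonTextBlock styled with a Fortnite typography style.
# The hardcoded text covers profile checks, MCP export and calendar offset for new content, and the limit to client-side quest, bundle and schedule views.
# The event graph holds only empty event stubs.
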WidgetBlueprint /QuestUI/DebugTool/Panels/Panel_CantFindAsset
graphs [EventGraph {nodes [
   Event PreConstruct()
   Event Construct()
   Event Tick()
  ]}]
variables []
widgets {
 info {inheritedWidgetCount=0, parentClass=&UserWidget, rootWidgetClass=&Overlay, widgetCount=2}
 widgets [
  Panel_Root {widgetClassPath=&Overlay, bIsVariable=true, widget=Panel_Root}
  CommonTextBlock_171 {
   widgetClassPath=&CommonTextBlock
   parent=Panel_Root
   widget=CommonTextBlock_171
   properties {
    autoWrapText=true
    style=/FortUILibrary/Styles/Typography/Foundation/TextStyle-Base-S-B.TextStyle-Base-S-B_C
    text="If the asset you are looking for is not in this list, this means it was not found in the current build.\r\n\r\n1. Verify the item exists on your profile. You can type \"mcpcheat\" in your console (~) to open your current profile.\r\n\r\nNote if you are looking for new content, MCP must be aware of it. If running Ephemeral MCP, you must export from UGS via More > MCP > Run (Local Content). If connected to an Epic backend instead, you will not see them until they are submitted and the build machine exports the latest templates.\r\n\r\nAlso ensure the backend is at the proper calendar offset to grant the content.\r\n\r\n2. This tool only reflects the quest, bundle, and schedule views that exist on your client. Various quest settings can impact the creation of a view, especially persisting a view after it is completed. If your asset is found but not live, it could be in a state where the view for it is not being created."
    font {
     bForceMonospaced=false
     bMaterialIsStencil=false
     fontMaterial=null
     fontObject=&NotoSans
     letterSpacing=0
     monospacedWidth=1
     size=18
     skewAmount=0
     typefaceFontName=NotoSans-Bold
     outlineSettings {
      bApplyOutlineToDropShadows=false
      bMiteredCorners=false
      bSeparateFillAlpha=false
      outlineMaterial=null
      outlineSize=0
      outlineColor=<rgba 0 0 0 1>
     }
    }
    shadowOffset=<0 0>
   }
   slot {class=OverlaySlot, properties {horizontalAlignment=HAlign_Center, verticalAlignment=VAlign_Center, padding=<ltrb 50 50 50 50>}}
  }
 ]
}
