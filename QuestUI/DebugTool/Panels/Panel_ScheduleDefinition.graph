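# Quest debug tool panel widget for displaying a schedule definition.
# A vertical box stacks a type header set to the ScheduleAsset entry type, then schedule info, plugin info and schedule entry info sub-widgets.
# The event graph is empty and the panel declares no variables of its own, so it serves as a layout composed from reusable debug tool widgets.
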WidgetBlueprint /QuestUI/DebugTool/Panels/Panel_ScheduleDefinition
graphs [EventGraph {links [], nodes []}]
variables []
widgets {
 info {inheritedWidgetCount=0, parentClass=&UserWidget, rootWidgetClass=&VerticalBox, widgetCount=5}
 widgets [
  Panel_Root {widgetClassPath=&VerticalBox, bIsVariable=true, widget=Panel_Root}
  TypeHeader {
   widgetClassPath=/QuestUI/DebugTool/CommonWidgets/QuestDebugToolTypeHeader
   parent=Panel_Root
   widget=TypeHeader
   properties {debugEntryType=ScheduleAsset}
   slot {class=VerticalBoxSlot}
  }
  Widget_ScheduleInfo {
   widgetClassPath=/QuestUI/DebugTool/ScheduleWidgets/QuestDebugToolScheduleInfo
   bIsVariable=true
   parent=Panel_Root
   widget=Widget_ScheduleInfo
   slot {class=VerticalBoxSlot}
  }
  Widget_BuildInfo {
   widgetClassPath=/QuestUI/DebugTool/CommonWidgets/QuestDebugToolPluginInfo
   bIsVariable=true
   parent=Panel_Root
   widget=Widget_BuildInfo
   slot {class=VerticalBoxSlot}
  }
  Widget_ScheduleEntryInfo {
   widgetClassPath=/QuestUI/DebugTool/ScheduleWidgets/QuestDebugToolScheduleEntryInfo
   bIsVariable=true
   parent=Panel_Root
   widget=Widget_ScheduleEntryInfo
   slot {class=VerticalBoxSlot}
  }
 ]
}
