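# Schedule view panel for the quest debug tool, a UserWidget in the QuestUI debug tool.
# A vertical box stacks a type header set to ScheduleView, then schedule info, plugin info and schedule entry info sub-widgets.
# The event graph holds only unconnected PreConstruct, Construct and Tick events, so the panel itself is pure layout composed of those widgets.
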WidgetBlueprint /QuestUI/DebugTool/Panels/Panel_ScheduleView
graphs [EventGraph {nodes [
   Event PreConstruct()
   Event Construct()
   Event Tick()
  ]}]
variables []
widgets {
 info {inheritedWidgetCount=0, parentClass=&UserWidget, rootWidgetClass=&VerticalBox, widgetCount=5}
 widgets [
  Panel_Root {widgetClassPath=&VerticalBox, bIsVariable=true, widget=Panel_Root}
  TypeHeader {
   widgetClassPath=/QuestUI/DebugTool/CommonWidgets/QuestDebugToolTypeHeader
   parent=Panel_Root
   widget=TypeHeader
   properties {debugEntryType=ScheduleView}
   slot {class=VerticalBoxSlot}
  }
  Widget_ScheduleInfo {
   widgetClassPath=/QuestUI/DebugTool/ScheduleWidgets/QuestDebugToolScheduleInfo
   bIsVariable=true
   parent=Panel_Root
   widget=Widget_ScheduleInfo
   slot {class=VerticalBoxSlot}
  }
  Widget_PluginInfo {
   widgetClassPath=/QuestUI/DebugTool/CommonWidgets/QuestDebugToolPluginInfo
   bIsVariable=true
   parent=Panel_Root
   widget=Widget_PluginInfo
   slot {class=VerticalBoxSlot}
  }
  Widget_ScheduleEntryInfo {
   widgetClassPath=/QuestUI/DebugTool/ScheduleWidgets/QuestDebugToolScheduleEntryInfo
   bIsVariable=true
   parent=Panel_Root
   widget=Widget_ScheduleEntryInfo
   slot {class=VerticalBoxSlot}
  }
 ]
}
