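# Quest UI debugger screen that lets developers filter and inspect quest, bundle, schedule and category entries and assets.
# The native QuestPanelDebugTool parent drives it through events, with filters and a results list beside a scrolling panel switcher.
# SetSwitcherState maps selection type to a detail panel, showing the building panel instead while a build is active.
# Search changes update the result count and show the "can't find asset" panel when nothing matches.
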WidgetBlueprint /QuestUI/DebugTool/QuestDebugTool
graphs [
 BP_BuildAllTick {nodes [
   n1 = BP_BuildAllTick!() -> (then, BuildState:EQuestPanelDebugToolBuildState Enum=Init, CurrentProgress:int=0, MaxProgress:int=0) {is=title K2Node_FunctionEntry}
   > Update!(
    execute
    self:Panel Building Object Reference=Get Panel_Building() -> (Panel_Building:Panel Building Object Reference)
    State:EQuestPanelDebugToolBuildState Enum=@n1
    CurrentProgress:int=@n1.CurrentProgress
    MaxProgress:int=@n1.MaxProgress
   ) -> (then)
   > Branch(Condition="Not Equal (Enum)"!(
     A:EQuestPanelDebugToolBuildState Enum=Get BuildState() -> (BuildState:EQuestPanelDebugToolBuildState Enum=Init)
     B:EQuestPanelDebugToolBuildState Enum=Get BuildAllState() -> (BuildAllState:EQuestPanelDebugToolBuildState Enum=Init)
    ) -> (ReturnValue:bool))
   > Set BuildAllState(BuildAllState:EQuestPanelDebugToolBuildState Enum=Get BuildState() -> (BuildState=Init)) -> (Output_Get:EQuestPanelDebugToolBuildState Enum=Init)
   > SetSwitcherState!(
    execute
    self
    SelectionType:EQuestPanelDebugListEntryType Enum=GetSelectionType!(self:Quest Panel Debug Tool Object Reference) -> (ReturnValue:EQuestPanelDebugListEntryType Enum=Unset)
   ) -> (then)
  ]}
 EventGraph {nodes [
   "On Clicked (Button_CantFindAsset)"!() -> (OutputDelegate, then, Button:Common Button Base Object Reference) {is=title K2Node_ComponentBoundEvent}
   > SetSwitcherState(SelectionType=NotFound)
   "On Clicked (Button_Close)"!() -> (OutputDelegate, then, Button:Common Button Base Object Reference) {is=title K2Node_ComponentBoundEvent}
   > DeactivateWidget()
   Event On Activated()
   n1 = Create Event()
   n2 = Get Filters() -> (Filters:Quest Debug Tool Filters Object Reference)
   > Bind Event to On Active Filter Changed(self:Quest Debug Tool Filters Object Reference=@n2, Delegate=@n1) {is=title K2Node_AddDelegate}
   n3 = Create Event()
   > Bind Event to On Build Tags Requested(self:Quest Debug Tool Filters Object Reference=@n2, Delegate=@n3) {is=title K2Node_AddDelegate}
   > InitFilter!(
    execute
    self:Quest Debug Tool Filters Object Reference=Get Filters()
    Filter:"Quest Panel Debug Tool Filter Structure (by ref)"=Make Quest Panel Debug Tool Filter!() -> (QuestPanelDebugToolFilter:Quest Panel Debug Tool Filter Structure) {is=title K2Node_MakeStruct}
    RootOrganizationNode:Quest Panel Debug Organization Node Object Reference
   ) -> (then)
   Event On Deactivated()
   > Unbind Event from On Active Filter Changed(self:Quest Debug Tool Filters Object Reference=@n2, Delegate=@n1) {is=title K2Node_RemoveDelegate}
   > Unbind Event from On Build Tags Requested(self:Quest Debug Tool Filters Object Reference=@n2, Delegate=@n3) {is=title K2Node_RemoveDelegate}
   Event OnInitialized()
   > SetSearchFilter!(
    execute
    self:Quest Panel Debug Tool Object Reference
    SearchFilter:"Quest Panel Debug Tool Filter Structure (by ref)"=Get ActiveFilter!(self:Quest Debug Tool Filters Object Reference=Get Filters()) -> (ActiveFilter:Quest Panel Debug Tool Filter Structure)
   ) -> (then)
   > SetSwitcherState(SelectionType=Unset)
   n4 = Event BP_SetSelectionType!() -> (OutputDelegate, then, EntryType:EQuestPanelDebugListEntryType Enum=Unset) {is=title K2Node_Event}
   > SetSwitcherState(SelectionType=@n4.EntryType)
   Event BP_SearchResultsChanged() {is=title K2Node_Event}
   n5 = GetNumItems!(self:List View Object Reference=Get ListView_Entries() -> (ListView_Entries:List View Object Reference)) -> (ReturnValue:int=0)
   > "SetText (Text)"(
    self=Get Text_Results() -> (Text_Results:Common Text Object Reference)
    InText=Format Text!(Format="LOCTEXT(\"{0} Results Found\")", 0:int=@n5) -> (Result)
   )
   > n6 = Branch(Condition="Equal (Integer)"(A=@n5))
   > SetSwitcherState(SelectionType=NotFound)
   n7 = Is Valid(
    exec=@n6.else
    InputObject=Get Selected Item!(self:List View Object Reference=Get ListView_Entries()) -> (ReturnValue:Object Reference) {is=title K2Node_CallFunction}
   )
   SetSwitcherState(execute=@n7.Is Not Valid, SelectionType=Unset)
   n8 = Event BP_Init!() -> (
    OutputDelegate
    then
    SearchFilter:"Quest Panel Debug Tool Filter Structure (by ref)"
    OrganizationNode:Quest Panel Debug Organization Node Object Reference
   ) {is=title K2Node_Event}
   > InitFilter(self=Get Filters(), Filter=@n8.SearchFilter, RootOrganizationNode=@n8.OrganizationNode)
  ]}
 SetSwitcherState {nodes [
   n1 = SetSwitcherState!() -> (then, SelectionType:EQuestPanelDebugListEntryType Enum=Unset) {is=title K2Node_FunctionEntry}
   > n2 = Branch(Condition="Equal (Enum)"(A:EQuestPanelDebugToolBuildState Enum=Get BuildAllState() -> (BuildAllState=Init), B:EQuestPanelDebugToolBuildState Enum=Inactive))
   > n3 = SetActiveWidget(
    self=Get ContentSwitcher() -> (ContentSwitcher:Common Visibility Switcher Object Reference)
    Widget=Select!(
     Unset:Widget Object Reference=Get Panel_NoSelection() -> (Panel_NoSelection:Panel No Selection Object Reference)
     NotFound:Widget Object Reference=Get Panel_CantFindAsset() -> (Panel_CantFindAsset:Panel Cant Find Asset Object Reference)
     ScheduleView:Widget Object Reference=Get Panel_ScheduleView() -> (Panel_ScheduleView:Panel Schedule View Object Reference)
     ScheduleAsset:Widget Object Reference=Get Panel_ScheduleAsset() -> (Panel_ScheduleAsset:Panel Schedule Definition Object Reference)
     BundleView:Widget Object Reference=Get Panel_BundleView() -> (Panel_BundleView:Panel Bundle View Object Reference)
     BundleAsset:Widget Object Reference=Get Panel_BundleAsset() -> (Panel_BundleAsset:Panel Bundle Definition Object Reference)
     QuestView:Widget Object Reference=Get Panel_QuestView() -> (Panel_QuestView:Panel Quest View Object Reference)
     QuestAsset:Widget Object Reference=Get Panel_QuestAsset() -> (Panel_QuestAsset:Panel Quest Definition Object Reference)
     CategoryInstance:Widget Object Reference=Get Panel_CategoryInstance() -> (Panel_CategoryInstance:Panel Category Instance Object Reference)
     CategoryAsset:Widget Object Reference=Get Panel_CategoryData() -> (Panel_CategoryData:Panel Category Data Object Reference)
     Index:EQuestPanelDebugListEntryType Enum=@n1
    ) -> (ReturnValue:Widget Object Reference)
   )
   SetIsEnabled(
    execute=[@n3.then, @n4.then]
    self=Get Button_CantFindAsset() -> (Button_CantFindAsset:WBP UIKit Button Quiet Object Reference)
    bInIsEnabled="Not Equal (Enum)"(A:EQuestPanelDebugListEntryType Enum=@n1, B:EQuestPanelDebugListEntryType Enum=NotFound)
   )
   > ScrollToStart(self:Scroll Box Object Reference=Get Scrollbox_Content() -> (Scrollbox_Content:Scroll Box Object Reference))
   n4 = SetActiveWidget(execute=@n2.else, self=Get ContentSwitcher(), Widget=Get Panel_Building())
  ]}
]
variables [BuildAllState {type=EQuestPanelDebugToolBuildState, default=Inactive}]
widgets {
 info {inheritedWidgetCount=1, parentClass=/Script/QuestUI.QuestPanelDebugTool, rootWidgetClass=&Overlay, widgetCount=28}
 widgets [
  Overlay_25 {widgetClassPath=&Overlay, widget=Overlay_25}
  BG {
   widgetClassPath=&Image
   parent=Overlay_25
   widget=BG
   properties {colorAndOpacity=<rgba 0 0.0036 0.0312 0.93>}
   slot {class=OverlaySlot, properties {horizontalAlignment=HAlign_Fill, verticalAlignment=VAlign_Fill}}
  }
  VerticalBox_0 {
   widgetClassPath=&VerticalBox
   parent=Overlay_25
   widget=VerticalBox_0
   slot {class=OverlaySlot, properties {horizontalAlignment=HAlign_Fill, verticalAlignment=VAlign_Fill, padding=<ltrb 20 20 20 20>}}
  }
  HorizontalBox_2 {
   widgetClassPath=&HorizontalBox
   parent=VerticalBox_0
   widget=HorizontalBox_2
   slot {class=VerticalBoxSlot, properties {padding=<ltrb 0 0 0 10>}}
  }
  Widget_RootHeader {
   widgetClassPath=/QuestUI/DebugTool/CommonWidgets/QuestDebugToolRootHeader
   bIsVariable=true
   parent=HorizontalBox_2
   widget=Widget_RootHeader
   slot {class=HorizontalBoxSlot, properties {size {sizeRule=Fill, value=1}}}
  }
  Button_Close {
   widgetClassPath=/FortUILibrary/UIKit/Buttons/WBP_UIKit_Button_Quiet
   bIsVariable=true
   parent=HorizontalBox_2
   widget=Button_Close
   properties {minWidth=32, text=X, toolTipText=Close the Quest UI Debugger}
   slot {class=HorizontalBoxSlot, properties {verticalAlignment=VAlign_Top, padding=<ltrb 32 0 0 0>}}
  }
  HorizontalBox_0 {
   widgetClassPath=&HorizontalBox
   parent=VerticalBox_0
   widget=HorizontalBox_0
   slot {class=VerticalBoxSlot, properties {size {sizeRule=Fill, value=1}}}
  }
  SizeBox_0 {
   widgetClassPath=&SizeBox
   parent=HorizontalBox_0
   widget=SizeBox_0
   properties {bOverride_WidthOverride=true, widthOverride=500}
   slot {class=HorizontalBoxSlot}
  }
  VerticalBox_31 {widgetClassPath=&VerticalBox, parent=SizeBox_0, widget=VerticalBox_31, slot {class=SizeBoxSlot}}
  Filters {
   widgetClassPath=/QuestUI/DebugTool/CommonWidgets/QuestDebugToolFilters
   bIsVariable=true
   parent=VerticalBox_31
   widget=Filters
   slot {class=VerticalBoxSlot}
  }
  Text_Results {
   widgetClassPath=&CommonTextBlock
   bIsVariable=true
   parent=VerticalBox_31
   widget=Text_Results
   properties {
    style=/FortUILibrary/Styles/Typography/Foundation/TextStyle-Base-S-B.TextStyle-Base-S-B_C
    text=999 Results Found
    font {
     bForceMonospaced=false
     bMaterialIsStencil=false
     fontMaterial=null
     fontObject=&NotoSans
     letterSpacing=0
     monospacedWidth=1
     size=18
     skewAmount=0
     typefaceFontName=NotoSans-Bold
     outlineSettings {
      bApplyOutlineToDropShadows=false
      bMiteredCorners=false
      bSeparateFillAlpha=false
      outlineMaterial=null
      outlineSize=0
      outlineColor=<rgba 0 0 0 1>
     }
    }
    shadowOffset=<0 0>
   }
   slot {class=VerticalBoxSlot, properties {horizontalAlignment=HAlign_Center, padding=<ltrb 0 10 0 0>}}
  }
  ListView_Entries {
   widgetClassPath=/Script/UMG.ListView
   bInherited=true
   bIsVariable=true
   parent=VerticalBox_31
   widget=ListView_Entries
   properties {entryWidgetClass=/QuestUI/DebugTool/CommonWidgets/QuestDebugToolListEntry.QuestDebugToolListEntry_C, verticalEntrySpacing=2}
   slot {class=VerticalBoxSlot, properties {size {sizeRule=Fill, value=1}}}
  }
  Spacer_64 {widgetClassPath=&Spacer, parent=VerticalBox_31, widget=Spacer_64, properties {size=<1 10>}, slot {class=VerticalBoxSlot}}
  Button_CantFindAsset {
   widgetClassPath=&WBP_UIKit_Button_Quiet
   bIsVariable=true
   parent=VerticalBox_31
   widget=Button_CantFindAsset
   properties {minWidth=120, text=I can't find my asset}
   slot {class=VerticalBoxSlot, properties {horizontalAlignment=HAlign_Center, verticalAlignment=VAlign_Top}}
  }
  Image_Spacer_Vert_1 {
   widgetClassPath=&Image
   parent=HorizontalBox_0
   widget=Image_Spacer_Vert_1
   properties {
    brush {
     bIsDynamicallyLoaded=false
     drawAs=Image
     imageType=NoImage
     mirroring=NoMirror
     resourceName=None
     resourceObject=null
     tiling=NoTile
     imageSize=<4 32>
     margin=<ltrb 0 0 0 0>
     outlineSettings {
      bUseBrushTransparency=false
      roundingType=HalfHeightRadius
      width=0
      color {colorUseRule=UseColor_Specified, specifiedColor=<rgba 0 0 0 0>}
      cornerRadii=<0 0 0 0>
     }
     tintColor {colorUseRule=UseColor_Specified, specifiedColor=<rgba 1 1 1 1>}
     uVRegion {bIsValid=false, max=<0 0>, min=<0 0>}
    }
    colorAndOpacity=<rgba 1 1 1 0.1>
   }
   slot {class=HorizontalBoxSlot, properties {padding=<ltrb 8 8 8 8>}}
  }
  Scrollbox_Content {
   widgetClassPath=&ScrollBox
   bIsVariable=true
   parent=HorizontalBox_0
   widget=Scrollbox_Content
   slot {class=HorizontalBoxSlot, properties {size {sizeRule=Fill, value=1}}}
  }
  ContentSwitcher {
   widgetClassPath=&CommonVisibilitySwitcher
   bIsVariable=true
   parent=Scrollbox_Content
   widget=ContentSwitcher
   slot {class=ScrollBoxSlot, properties {size {sizeRule=Fill, value=1}}}
  }
  Panel_Building {
   widgetClassPath=/QuestUI/DebugTool/Panels/Panel_Building
   bIsVariable=true
   parent=ContentSwitcher
   widget=Panel_Building
   slot {class=CommonVisibilitySwitcherSlot}
  }
  Panel_CantFindAsset {
   widgetClassPath=/QuestUI/DebugTool/Panels/Panel_CantFindAsset
   bIsVariable=true
   parent=ContentSwitcher
   widget=Panel_CantFindAsset
   slot {class=CommonVisibilitySwitcherSlot}
  }
  Panel_NoSelection {
   widgetClassPath=/QuestUI/DebugTool/Panels/Panel_NoSelection
   bIsVariable=true
   parent=ContentSwitcher
   widget=Panel_NoSelection
   slot {class=CommonVisibilitySwitcherSlot}
  }
  Panel_QuestView {
   widgetClassPath=/QuestUI/DebugTool/Panels/Panel_QuestView
   bIsVariable=true
   parent=ContentSwitcher
   widget=Panel_QuestView
   slot {class=CommonVisibilitySwitcherSlot}
  }
  Panel_QuestAsset {
   widgetClassPath=/QuestUI/DebugTool/Panels/Panel_QuestDefinition
   bIsVariable=true
   parent=ContentSwitcher
   widget=Panel_QuestAsset
   slot {class=CommonVisibilitySwitcherSlot}
  }
  Panel_BundleView {
   widgetClassPath=/QuestUI/DebugTool/Panels/Panel_BundleView
   bIsVariable=true
   parent=ContentSwitcher
   widget=Panel_BundleView
   slot {class=CommonVisibilitySwitcherSlot}
  }
  Panel_BundleAsset {
   widgetClassPath=/QuestUI/DebugTool/Panels/Panel_BundleDefinition
   bIsVariable=true
   parent=ContentSwitcher
   widget=Panel_BundleAsset
   slot {class=CommonVisibilitySwitcherSlot}
  }
  Panel_ScheduleView {
   widgetClassPath=/QuestUI/DebugTool/Panels/Panel_ScheduleView
   bIsVariable=true
   parent=ContentSwitcher
   widget=Panel_ScheduleView
   slot {class=CommonVisibilitySwitcherSlot}
  }
  Panel_ScheduleAsset {
   widgetClassPath=/QuestUI/DebugTool/Panels/Panel_ScheduleDefinition
   bIsVariable=true
   parent=ContentSwitcher
   widget=Panel_ScheduleAsset
   slot {class=CommonVisibilitySwitcherSlot}
  }
  Panel_CategoryInstance {
   widgetClassPath=/QuestUI/DebugTool/Panels/Panel_CategoryInstance
   bIsVariable=true
   parent=ContentSwitcher
   widget=Panel_CategoryInstance
   slot {class=CommonVisibilitySwitcherSlot}
  }
  Panel_CategoryData {
   widgetClassPath=/QuestUI/DebugTool/Panels/Panel_CategoryData
   bIsVariable=true
   parent=ContentSwitcher
   widget=Panel_CategoryData
   slot {class=CommonVisibilitySwitcherSlot}
  }
 ]
}
